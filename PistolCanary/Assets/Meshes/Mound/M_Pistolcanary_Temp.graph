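# Temporary Default Lit surface material for the PistolCanary debris mound mesh, with Interchange import data attached.
# It holds three texture parameters, BaseColor, Normal and a masks-sampled ORM, defaulting to the PistolCanaryDebris textures.
# The graph shows no links from these samplers to material outputs and no other math.
# Usage flags enable it for Nanite and instanced static meshes.
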Material /PistolCanary/Assets/Meshes/Mound/M_Pistolcanary_Temp
graph {
 properties {
  assetImportData="<self>:InterchangeAssetImportData_0"
  bUsedWithInstancedStaticMeshes=true
  bUsedWithNanite=true
  usedShadingModels=Default Lit
 }
 nodes [
  TextureSampleParameter2D() {parameterName=BaseColor, texture=/PistolCanary/Assets/Meshes/Mound/Textures/T_PistolCanaryDebris_BaseColor}
  TextureSampleParameter2D() {parameterName=Normal, samplerType=SAMPLERTYPE_Normal, texture=/PistolCanary/Assets/Meshes/Mound/Textures/T_PistolCanaryDebris_Normal}
  TextureSampleParameter2D() {parameterName=ORM, samplerType=SAMPLERTYPE_Masks, texture=/PistolCanary/Assets/Meshes/Mound/Textures/T_PistolCanaryDebris_ORM}
 ]
}
